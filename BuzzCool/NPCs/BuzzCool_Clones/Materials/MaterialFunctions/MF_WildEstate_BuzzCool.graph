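# Material function for BuzzCool WildEstate that supplies anime-style shading inputs as separate outputs plus a combined MaterialAttributes output.
# Textures and parameters, with an optional CustomUV input, provide lit color, darkened shaded color, FX mask, STT, line, rim and specular colors.
# A static switch picks SRM map or scalars, but its B channel alone drives Metallic, Specular and Roughness; the normal is flat.
# The specular-map switch and both cinematic tint calls feed no output.
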MaterialFunction /BuzzCool/NPCs/BuzzCool_Clones/Materials/MaterialFunctions/MF_WildEstate_BuzzCool
graph {nodes [
  n1 = Multiply(
   A=NamedRerouteUsage()
   B=ScalarParameter() {defaultValue=0.5, group=01_Anime Maps, parameterName=Dynamic_ColorShaded_Darkening, sliderMax=1, sortPriority=11}
  )
  FunctionOutput(None=@n1) {outputName=Color_Shaded, sortPriority=12}
  n2 = VectorParameter() {desc="RGB=Color", group=01_Anime Maps, parameterName=StyleRimLight_Color, sortPriority=22, defaultValue=<rgba 1 1 1 1>}
  FunctionOutput(None=@n2.RGB) {outputName=StyleRimLight_Color, sortPriority=17}
  FunctionOutput(None=MaterialFunctionCall(
    X=ScalarParameter() {group=01_Anime Maps, parameterName=STT_Sheen, sliderMax=1, sortPriority=21}
    Y=ScalarParameter() {defaultValue=0.5, group=01_Anime Maps, parameterName=STT_Threshold, sliderMax=1, sortPriority=22}
    Z=ScalarParameter() {defaultValue=1, group=01_Anime Maps, parameterName=STT_Thickness, sliderMax=1, sortPriority=23}
   ) {materialFunction=&MakeFloat3}.Result) {outputName=STT, sortPriority=1}
  n3 = VectorParameter() {group=01_Anime Maps, parameterName=DFLine_Color, sortPriority=20, defaultValue=<rgba 0.009 0.0019 0.0208 1>}
  FunctionOutput(None=Saturate(None=@n3.RGB)) {outputName=DFL_Color, sortPriority=15}
  n4 = TextureSampleParameter2D(UVs=NamedRerouteUsage()) {
   group=01_Anime Maps
   parameterName=FX_Map
   samplerType=SAMPLERTYPE_Masks
   sortPriority=16
   texture=&T_BlackMask
  }
  FunctionOutput(None=@n4.RGBA) {outputName=FX, sortPriority=4}
  n5 = TextureSampleParameter2D(UVs=NamedRerouteUsage()) {group=01_Anime Maps, parameterName=Color_Lit_Map, sortPriority=10, texture=&T_WhitesRGB}
  n6 = NamedRerouteDeclaration(None=@n5.RGB) {name=Color_Lit, nodeColor=<rgba 0 0.4609 1 1>}
  n7 = TextureSampleParameter2D(UVs=NamedRerouteUsage()) {
   desc="R=Spec\r\nG=Rough\r\nB=Metallic"
   group=01_Anime Maps
   parameterName=SRM_Map
   samplerType=SAMPLERTYPE_Masks
   sortPriority=15
   texture=&T_Fortnite_Default_S
   channelNames {a="", b=Metallic, g=Roughness, r=Specular}
  }
  n8 = MaterialFunctionCall(
   X=ScalarParameter() {defaultValue=0.5, group=01_Anime Maps, parameterName=SRM_Specular, sliderMax=1, sortPriority=15}
   Y=ScalarParameter() {defaultValue=0.5, group=01_Anime Maps, parameterName=SRM_Roughness, sliderMax=1, sortPriority=15}
   Z=ScalarParameter() {group=01_Anime Maps, parameterName=SRM_Metallic, sliderMax=1, sortPriority=15}
  ) {materialFunction=&MakeFloat3}
  n9 = MaterialFunctionCall(Float3=FunctionOutput(None=StaticSwitchParameter(True=@n7.RGB, False=@n8.Result) {group=01_Anime Maps, parameterName=UseSRMMap, sortPriority=5}) {outputName=SRM, sortPriority=13}) {materialFunction=&BreakOutFloat3Components}
  n10 = Constant3Vector() {constant=<rgba 0 0 1 0>}
  FunctionOutput(None=MakeMaterialAttributes!(
    BaseColor=FunctionOutput(None=@n6) {outputName=Color_Lit, sortPriority=11}
    Metallic=@n9.B
    Specular=@n9.B
    Roughness=@n9.B
    Anisotropy
    EmissiveColor
    Opacity
    OpacityMask
    Normal=FunctionOutput(None=@n10) {outputName=Normal, sortPriority=2}
    Tangent
    WorldPositionOffset
    SubsurfaceColor
    ClearCoat
    ClearCoatRoughness
    AmbientOcclusion
    Refraction
    CustomizedUVs_0
    CustomizedUVs_1
    CustomizedUVs_2
    CustomizedUVs_3
    CustomizedUVs_4
    CustomizedUVs_5
    CustomizedUVs_6
    CustomizedUVs_7
    PixelDepthOffset
    ShadingModel
    Displacement
   )) {outputName=--------, sortPriority=10}
  n11 = VectorParameter() {
   desc="RGB=Color, A=Opacity"
   group=01_Anime Maps
   parameterName=Specular_Color
   sortPriority=21
   channelNames {a=Opacity, b=Blue Channel, g=Green Channel, r=Red Channel}
   defaultValue=<rgba 1 1 1 1>
  }
  n12 = AppendVector(A=@n11, B=@n11)
  FunctionOutput(None=@n12) {outputName=Specular_Color, sortPriority=16}
  MaterialFunctionCall!(LitBaseColorIn) {materialFunction=/Game/Packages/Fortress_SharedMaterials/Cinematics/Anime/MF_CINE_LitDiffuseTint}
  MaterialFunctionCall!(ShadedBaseColorIn=@n1) {materialFunction=/Game/Packages/Fortress_SharedMaterials/Cinematics/Anime/MF_CINE_ShadedDiffuseTint}
  NamedRerouteDeclaration(None=FunctionInput(Preview=TextureCoordinate()) {bUsePreviewValueAsDefault=true, inputName=CustomUV, inputType=FunctionInput_Vector2}) {name=UV Coords, nodeColor=<rgba 0.5 0.5 0 1>}
  n13 = TextureSampleParameter2D(UVs=NamedRerouteUsage()) {group=01_Anime Maps, parameterName=SpecularColor_Map, sortPriority=19, texture=&T_BlacksRGB}
  StaticSwitchParameter(
   True=AppendVector(A=@n13.RGB, B=ScalarParameter() {defaultValue=1, group=01_Anime Maps, parameterName=SpecularColor_Opacity, sliderMax=1, sortPriority=19})
   False=@n12
  ) {group=01_Anime Maps, parameterName=UseSpecularColorMap, sortPriority=9}
 ]}
