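# Character material for the BuzzCool clone NPC skeletal mesh, layering an animated galaxy-style emissive effect over a base material function.
# Camera-vector triplanar projections build rotating nebula and twinkling stars, plus noise-warped waves colored through a curve atlas.
# Fresnel terms shape a flame-like rim, and view-angle gradients color the glowing eyes.
# Blended attributes, optional flat shading and the shared character parent post-FX function form the remaining stages.
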Material /BuzzCool/NPCs/BuzzCool_Clones/Materials/Version1/M_WildEstate_BuzzCool_V1
graph {
 properties {
  bAutomaticallySetUsageInEditor=false
  bUseMaterialAttributes=true
  bUsedWithMeshDeformer=true
  bUsedWithSkeletalMesh=true
  usedShadingModels=Default Lit
 }
 nodes [
  n1 = VertexNormalWS()
  n2 = ScalarParameter() {defaultValue=0.0725, group=GalaxyV2, parameterName=WavesNoise_RotationSpeed, sliderMax=1}
  n3 = ScalarParameter() {defaultValue=-0.0725, group=GalaxyV2, parameterName=WavesNoise_RotationSpeed_1, sliderMax=1}
  n4 = ScalarParameter() {defaultValue=6, group=GalaxyV2, parameterName=Galaxy_AxisFade, sliderMax=8}
  n5 = MaterialFunctionCall!(
   CameraVector=NamedRerouteUsage()
   Rotation Speed=AppendVector(A=AppendVector(A=@n2, B=@n3), B=ScalarParameter() {defaultValue=0.1, group=GalaxyV2, parameterName=WavesNoise_RotationSpeed_2, sliderMax=1})
   Texture=TextureObject() {samplerType=SAMPLERTYPE_LinearGrayscale, texture=/Game/Global/Textures/Default/Noise/T_SwirlNoise_N_height}
   Tiling=ScalarParameter() {defaultValue=1, group=GalaxyV2, parameterName=WavesNoise_Tiling, sliderMax=2}
   Offset
   AxisFadeContrast=@n4
  ) {materialFunction=/Game/Weapons/WeaponSkins/MaterialFunction/TriplanarCameraVector_Rotated}
  n6 = Add(
   A=Add(A=ScalarParameter() {defaultValue=4, group=GalaxyV2, parameterName=Waves_WarpIntensity, sliderMax=1}, B=NamedRerouteUsage())
   B=Multiply(
    A=ComponentMask(None=@n5.Result) {r=true}
    B=ScalarParameter() {defaultValue=2.5, group=GalaxyV2, parameterName=WavesNoise_Amount, sliderMax=2}
   ) {constB=0.1}
  )
  Add(A=@n1, B=@n6)
  n7 = GetMaterialAttributes(None=MaterialFunctionCall!(CustomUV) {materialFunction=/BuzzCool/NPCs/BuzzCool_Clones/Materials/MaterialFunctions/MF_WildEstate_BuzzCool}.--------) {attributeGetTypes [
    #1
    #2
    #3
    #4
    #5
   ]}
  n8 = NamedRerouteDeclaration(None=@n7.Normal) {name=InNormal, nodeColor=<rgba 0 1 0.2969 1>}
  n9 = Multiply(A=ScalarParameter() {defaultValue=1, group=GalaxyV2, parameterName=Galaxy_GlobalBrightness, sliderMax=1}, B=NamedRerouteUsage())
  n10 = SetMaterialAttributes!(
   MaterialAttributes
   Base Color=Constant()
   Specular=ScalarParameter() {group=GalaxyV2, parameterName=Galaxy_Specular, sliderMax=1}
   Roughness=Multiply(A=ScalarParameter() {defaultValue=1, group=GalaxyV2, parameterName=Galaxy_Roughness, sliderMax=1}, B=@n7.Roughness)
   Normal=@n8
   Emissive Color=StaticSwitch(
    True=Add(A=@n7.Emissive Color, B=@n9)
    False=@n9
    Value=StaticBoolParameter() {desc="false = overwrite emissive", group=GalaxyV2, parameterName=UseAdditiveEmissiveGalaxy, sortPriority=2}
   )
  ) {attributeSetTypes [
    #4
    #5
    #1
    #2
    #3
   ]}
  n11 = SetMaterialAttributes(
   MaterialAttributes=@n7.MaterialAttributes
   Base Color=@n7.MaterialAttributes
   Specular=@n7.MaterialAttributes
   Roughness=@n7.MaterialAttributes
   Normal=@n8
   Emissive Color=NamedRerouteUsage()
  ) {attributeSetTypes [
    #4
    #5
    #1
    #2
    #3
   ]}
  BlendMaterialAttributes(A=@n10, B=@n11, Alpha=NamedRerouteUsage()) {vertexAttributeBlendType=UseA}
  Composite() {subgraphName=Construct Base Inputs}
  n12 = NamedRerouteUsage()
  n13 = Time()
  n14 = Constant()
  n15 = ScalarParameter() {defaultValue=0.004, group=GalaxyV2, parameterName=Waves_RotationSpeed, sliderMax=1}
  n16 = ScalarParameter() {defaultValue=-0.004, group=GalaxyV2, parameterName=Waves_RotationSpeed_1, sliderMax=1}
  n17 = Multiply(
   A=LinearInterpolate(
    A=@n13
    B=Multiply(A=@n13, B=@n14)
    Alpha=ScalarParameter() {defaultValue=0.0025, group=GalaxyV2, parameterName=Waves_BackgroundRotation, sliderMax=1}
   )
   B=AppendVector(A=AppendVector(A=@n15, B=@n16), B=ScalarParameter() {defaultValue=0.022, group=GalaxyV2, parameterName=Waves_RotationSpeed_2, sliderMax=1})
  )
  n18 = TextureObject() {samplerType=SAMPLERTYPE_LinearGrayscale, texture=&T_SwirlNoise_N_height}
  n19 = ScalarParameter() {defaultValue=2, group=GalaxyV2, parameterName=Waves_Tiling, sliderMax=2}
  n20 = Saturate(None=Desaturation(None=MaterialFunctionCall(CameraVector=@n12, Rotation Speed=@n17, Texture=@n18, Tiling=@n19, Offset=@n6, AxisFadeContrast=@n4) {materialFunction=&TriplanarCameraVector_Rotated}.Result))
  n21 = Constant() {r=5}
  Divide(A=Floor(None=Multiply(A=@n20, B=@n21) {constB=5}), B=Subtract(A=@n21))
  LinearInterpolate(A=@n1, B=Constant3Vector() {constant=<rgba 0 0 1 0>}) {constAlpha=0.99}
  n22 = Constant()
  MaterialExpressionReroute_4 = Reroute(None=@n22)
  n23 = PreSkinnedPosition()
  n24 = ShaderStageSwitch(PixelShader=VertexInterpolator(VS=@n23).PS, VertexShader=@n23)
  MaterialExpressionReroute_5 = Reroute(None=@n24)
  n25 = MaterialFunctionCall!(Material) {materialFunction=/Game/Packages/Fortress_SharedMaterials/Base_Material_Functions/CharacterBrightnessandScale_Balance}
  MaterialExpressionReroute_6 = Reroute(None=@n25.Distance Blend)
  n26 = Fresnel() {baseReflectFraction=0, exponent=1}
  MaterialExpressionReroute_7 = Reroute(None=@n26)
  MaterialFunctionCall(
   Material=MaterialFunctionCall!(
    Material In=NamedRerouteUsage()
    UseFlatShading=StaticBoolParameter() {group=FlatShading, parameterName=UseFlatShading_All, sortPriority=1}
    IsWorldSpace=StaticBool() {value=true}
   ) {materialFunction=/BuzzCool/NPCs/BuzzCool_Clones/Materials/MaterialFunctions/MF_WildEstate_BuzzCool_FlatShading}.Result
   M=@MaterialExpressionReroute_4.M
   Pre Skinned Local Position=@MaterialExpressionReroute_5.Pre Skinned Local Position
   Distance Blend=@MaterialExpressionReroute_6.Distance Blend
   Fresnel=@MaterialExpressionReroute_7.Fresnel
  ) {materialFunction=&MF_CharacterParent_2023_PostFX}
  Multiply(A=NamedRerouteUsage()) {constB=8}
  Multiply(A=@n19, B=Constant3Vector() {constant=<rgba 3 1.33 0.2 0>})
  n27 = VectorParameter() {group=GalaxyV2, parameterName=EdgeFresnel, defaultValue=<rgba 0.0278 1 0 1>}
  NamedRerouteDeclaration(None=@n27.RGB) {name=Fresnel edge, nodeColor=<rgba 0.9922 1 0 1>}
  n28 = TextureSampleParameter2D() {group=BuzzCool_Eyes, parameterName=Eyes_Mask, samplerType=SAMPLERTYPE_Masks, texture=&T_BlackMask}
  NamedRerouteDeclaration(None=@n28.R) {name=SkinFX_Mask, nodeColor=<rgba 1 0.6562 0 1>}
  n29 = Fresnel(ExponentIn=ScalarParameter() {defaultValue=3, group=GalaxyV2, parameterName=Galaxy_FresEX, sliderMax=8}) {baseReflectFraction=0}
  NamedRerouteDeclaration(None=Saturate(None=@n29) {desc=Wormhole Fresnel}) {name=MasterFrensel, nodeColor=<rgba 0.0078 1 0 1>}
  n30 = LinearInterpolate(Alpha=NamedRerouteUsage()) {constA=1, constB=0}
  n31 = MaterialFunctionCall() {materialFunction=/Game/Effects/Niagara/MasterMaterials/Functions/InstanceBounds_0-1_UVW}
  n32 = Power(Base=@n31.B)
  n33 = Add(A=OneMinus(None=@n31.B)) {constB=0.5}
  n34 = Multiply(
   A=Multiply(B=NamedRerouteDeclaration(None=@n32) {name=BGradient, nodeColor=<rgba 0.9453 1 0 1>}) {constA=0.15}
   B=LinearInterpolate(Alpha=Power(Base=@n33) {constExponent=1.2}) {constA=0.08}
  )
  NamedRerouteDeclaration(None=Multiply(A=Power(Base=@n30) {constExponent=8}, B=@n34)) {name=CharGradient, nodeColor=<rgba 1 0.0469 0 1>}
  n35 = MaterialFunctionCall(
   Input=MaterialFunctionCall() {materialFunction=/Engine/Functions/Engine_MaterialFunctions02/Utility/FOV}."FOV "
   Input Low=Constant() {r=0.414}
   Input High=Constant() {r=1.091}
   Target Low=Constant()
   Target High=Constant() {r=1}
  ) {materialFunction=&RemapValueRange}
  NamedRerouteDeclaration(None=LinearInterpolate(A=Constant() {r=3}, B=Constant() {r=1}, Alpha=Saturate(None=@n35.Result))) {name=FOVCorrection, nodeColor=<rgba 1 0 0.5391 1>}
  n36 = TextureObjectParameter() {
   group=GalaxyV2
   parameterName=StarsTexture
   samplerType=SAMPLERTYPE_Masks
   sortPriority=3
   texture=/Game/Characters/Player/Male/Medium/Bodies/M_Med_Soldier_04/Skins/Galaxy/Textures/Celestial_T_Stars
  }
  n37 = TextureSample(UVs=Panner(Coordinate=TextureCoordinate() {uTiling=2, vTiling=2}) {speedX=0.133, speedY=0.2}, Tex=@n36) {
   samplerSource=SSM_Wrap_WorldGroupSettings
   samplerType=SAMPLERTYPE_Masks
   texture=&Celestial_T_Stars
  }
  NamedRerouteDeclaration(None=@n37.G) {name=MaskingWave, nodeColor=<rgba 1 0.7031 0 1>}
  NamedRerouteDeclaration(None=@n34) {name=CharBGradient, nodeColor=<rgba 0 0.6016 1 1>}
  n38 = Time()
  n39 = Multiply(A=@n38, B=Constant())
  n40 = ScalarParameter() {defaultValue=0.025, group=GalaxyV2, parameterName=BackgroundRotation, sliderMax=1}
  n41 = AppendVector(
   A=ScalarParameter() {defaultValue=0.2, group=GalaxyV2, parameterName=Galaxy_RotationSpeed, sliderMax=1}
   B=ScalarParameter() {defaultValue=-0.2, group=GalaxyV2, parameterName=Galaxy_RotationSpeed_1, sliderMax=1}
  )
  n42 = ScalarParameter() {defaultValue=0.1, group=GalaxyV2, parameterName=Galaxy_RotationSpeed_2, sliderMax=1}
  n43 = TextureObjectParameter() {
   group=GalaxyV2
   parameterName=GalaxyTexture
   sortPriority=3
   texture=/Game/Characters/Player/Male/Medium/Bodies/M_Med_Soldier_04/Skins/Galaxy/Textures/T_FN_Nebula_style2
  }
  n44 = VectorParameter() {group=GalaxyV2, parameterName=Galaxy_Offset, defaultValue=<rgba 0 0 0 1>}
  n45 = MaterialFunctionCall(
   CameraVector=NamedRerouteUsage()
   Rotation Speed=Multiply(A=LinearInterpolate(A=@n38, B=@n39, Alpha=@n40), B=AppendVector(A=@n41, B=@n42))
   Texture=@n43
   Tiling=Multiply(A=ScalarParameter() {defaultValue=1, group=GalaxyV2, parameterName=Galaxy_Tiling, sliderMax=2}, B=NamedRerouteUsage())
   Offset=Add(A=ScalarParameter() {defaultValue=1, group=GalaxyV2, parameterName=WarpIntensity, sliderMax=1}, B=@n44.RGB)
   AxisFadeContrast=ScalarParameter() {defaultValue=6, group=GalaxyV2, parameterName=Galaxy_AxisFade, sliderMax=8}
  ) {materialFunction=&TriplanarCameraVector_Rotated}
  n46 = AppendVector(
   A=ScalarParameter() {group=GalaxyV2, parameterName=EdgeFresnel_GradientSpeed_X, sliderMax=2}
   B=ScalarParameter() {defaultValue=0.07, group=GalaxyV2, parameterName=EdgeFresnel_GradientSpeed_Y, sliderMax=2}
  )
  n47 = ScalarParameter() {defaultValue=1, group=GalaxyV2, parameterName=EdgeFresnel_Tiling, sliderMax=4}
  n48 = MaterialFunctionCall(Float3=MaterialFunctionCall!(
    CameraVector=NamedRerouteUsage()
    Time Scale
    Panner Speed=@n46
    Texture=TextureObject() {samplerType=SAMPLERTYPE_LinearColor, texture=/Game/Weapons/WeaponSkins/Textures/T_Swirly_Noise_highContrast1}
    Tiling=@n47
    Offset
    AxisFadeContrast
   ) {materialFunction=/Game/Weapons/WeaponSkins/MaterialFunction/TriplanarCamerVector_Panning}.Result) {materialFunction=&BreakOutFloat3Components}
  n49 = TextureCoordinate() {uTiling=2, vTiling=2}
  n50 = TextureSample(UVs=Panner(Coordinate=Multiply(A=@n49, B=@n47), Speed=@n46)) {
   samplerSource=SSM_Wrap_WorldGroupSettings
   samplerType=SAMPLERTYPE_LinearColor
   texture=&T_Swirly_Noise_highContrast1
  }
  n51 = LinearInterpolate(
   A=Multiply(A=ScalarParameter() {defaultValue=8, group=GalaxyV2, parameterName=AccentRimBrightness, sliderMax=1}, B=@n45.Result)
   B=@n27.RGB
   Alpha=Multiply(
    A=StaticSwitchParameter(True=@n48.R, False=@n50.R) {defaultValue=true, group=GalaxyV2, parameterName=UseTriplanerFresnelFlame, sortPriority=2}
    B=ScalarParameter() {defaultValue=1, group=GalaxyV2, parameterName=Galaxy_FlameFresneIntensity, sliderMax=1}
   )
  )
  NamedRerouteDeclaration(None=MaterialFunctionCall!(In=FeatureLevelSwitch(Default=@n51, ES2=@n27.RGB, ES3_1=@n27.RGB, SM4_REMOVED=@n27.RGB)) {materialFunction=/Game/Packages/Fortress_SharedMaterials/Anime_Master_Material/MF_TODCustomColors}.Result) {name=EdgeFresnel, nodeColor=<rgba 0.4297 1 0 1>}
  n52 = DotProduct(A=CameraVectorWS(), B=PixelNormalWS())
  n53 = LinearInterpolate(
   A=VectorParameter() {group=BuzzCool_Eyes, parameterName=BuzzCool_Eyes_BaseEmissive, defaultValue=<rgba 0.0262 0.9823 0 1>}.RGB
   B=VectorParameter() {group=BuzzCool_Eyes, parameterName=BuzzCool_Eyes_BaseColor, defaultValue=<rgba 0.019 0.1615 0.0102 1>}.RGB
   Alpha=Power(Base=@n52) {constExponent=1.25}
  )
  n54 = VectorParameter() {group=BuzzCool_Eyes, parameterName=BuzzCool_Eyes_CoreEmissive, defaultValue=<rgba 0.4285 1 0.4029 1>}
  n55 = Saturate(None=Power(Base=@n52, Exp=Constant() {r=2}))
  NamedRerouteDeclaration(None=Multiply(
    A=LinearInterpolate(A=@n53, B=@n54.RGB, Alpha=@n55)
    B=ScalarParameter() {defaultValue=1, group=BuzzCool_Eyes, parameterName=BushHerb_EmissiveIntensity}
   )) {name=Eyes, nodeColor=<rgba 0 1 0.1852 1>}
  n56 = Saturate(None=Power(Base=@n29) {constExponent=1.75})
  n57 = Desaturation(None=NamedRerouteUsage())
  n58 = Constant3Vector() {constant=<rgba 0.218 0.0132 0.6771 1>}
  n59 = NamedRerouteUsage()
  n60 = Power(Base=OneMinus(None=NamedRerouteUsage()))
  n61 = LinearInterpolate(
   A=MaterialFunctionCall!(Blend=@n59, Base=Constant3Vector() {constant=<rgba 0.2423 0.5271 0.9473 1>}) {materialFunction=/Engine/Functions/Engine_MaterialFunctions03/Blends/Blend_Screen}.Result
   B=@n59
   Alpha=Saturate(None=LinearInterpolate(Alpha=@n60) {constA=1, constB=0.6})
  )
  n62 = Add(
   A=Multiply(
    A=LinearInterpolate(Alpha=@n57) {constA=0.3, constB=1.5}
    B=MaterialFunctionCall(Blend=@n58, Base=@n61) {materialFunction=&Blend_Screen}.Result
   )
   B=NamedRerouteUsage()
  )
  n63 = LinearInterpolate(A=@n62, B=NamedRerouteUsage(), Alpha=NamedRerouteUsage())
  n64 = NamedRerouteUsage()
  n65 = Saturate(None=LinearInterpolate(Alpha=Multiply(A=@n64) {constB=8}) {constA=0.2})
  n66 = LinearInterpolate(A=LinearInterpolate(A=LinearInterpolate(A=@n62, B=@n63, Alpha=@n65), Alpha=@n56) {constB=0}, B=NamedRerouteUsage(), Alpha=@n56)
  n67 = NamedRerouteUsage()
  n68 = Power(Base=@n29) {constExponent=3}
  NamedRerouteDeclaration(None=LinearInterpolate(A=LinearInterpolate(A=@n66, B=@n67, Alpha=@n68), B=NamedRerouteUsage(), Alpha=NamedRerouteUsage())) {name=GalaxyEmissive, nodeColor=<rgba 1 0.4219 0 1>}
  NamedRerouteDeclaration(None=@n10) {name=Final, nodeColor=<rgba 0 0.3437 1 1>}
  n69 = Constant3Vector() {constant=<rgba 0 1 0 0>}
  n70 = Divide(A=Multiply(
    A=VectorParameter() {desc="Dont Touch! this Vector is set by the CCPM BP.", group=GalaxyV2, parameterName=ActorRotationAmount, sortPriority=64}.B
    B=ScalarParameter() {defaultValue=0.1, group=GalaxyV2, parameterName=RotationAmount, sliderMax=1}
   ) {constB=0.333}) {constB=360}
  n71 = Constant3Vector()
  n72 = Transform(None=CameraVectorWS()) {transformSourceType=TRANSFORMSOURCE_World, transformType=TRANSFORM_Local}
  n73 = MaterialFunctionCall(Float3=Add(A=RotateAboutAxis(NormalizedRotationAxis=@n69, RotationAngle=@n70, PivotPoint=@n71, Position=@n72), B=@n72)) {materialFunction=&BreakOutFloat3Components}
  NamedRerouteDeclaration(None=Transform(None=MaterialFunctionCall(X=@n73.B, Y=@n73.B, Z=@n73.B) {materialFunction=&MakeFloat3}.Result) {transformSourceType=TRANSFORMSOURCE_World, transformType=TRANSFORM_Local}) {name=CameraVector, nodeColor=<rgba 1 0.0833 0.3367 1>}
  NamedRerouteDeclaration(None=@n7.Base Color) {name=BaseColor_In, nodeColor=<rgba 0 0.8828 1 1>}
  n74 = CurveAtlasRowParameter(CurveTime=@n20) {
   atlas=/WildEstateTech/Gameplay/Materials/ColorCurves/CA_WildEstate_Tech
   curve=/WildEstateTech/Gameplay/Materials/ColorCurves/CC_WE_BuzzCool_01
   group=BuzzCool_Waves
   parameterName=CC_BuzzCool
  }
  NamedRerouteDeclaration(None=@n74) {name=WavesColor, nodeColor=<rgba 1 0.4688 0 1>}
  n75 = Saturate(None=Add(A=NamedRerouteUsage()) {constB=0.55})
  NamedRerouteDeclaration(None=LinearInterpolate(A=Multiply(A=@n74.A, B=@n75), Alpha=Fresnel() {baseReflectFraction=0, exponent=2}) {constB=0}) {name=WavesA, nodeColor=<rgba 0.1484 0 1 1>}
  n76 = Desaturation(None=@n45.Result, Fraction=ScalarParameter() {defaultValue=0.5, group=GalaxyV2, parameterName=StarSaturation, sliderMax=1})
  n77 = Multiply(A=@n76, B=VectorParameter() {group=GalaxyV2, parameterName=StarBrightness_Color1, defaultValue=<rgba 100 100 100 1>}.RGB)
  n78 = Time()
  n79 = ScalarParameter() {defaultValue=-0.05, group=GalaxyV2, parameterName=SmallStar_RotationSpeed, sliderMax=1, sliderMin=-1}
  n80 = ScalarParameter() {defaultValue=3, group=GalaxyV2, parameterName=SmallStar_Tiling, sliderMax=8}
  n81 = NamedRerouteUsage()
  n82 = ComponentMask(None=MaterialFunctionCall(CameraVector=NamedRerouteUsage(), Rotation Speed=Multiply(A=@n78, B=@n79), Texture=@n36, Tiling=Multiply(A=@n80, B=@n81), Offset=@n44.RGB) {materialFunction=&TriplanarCameraVector_Rotated}.Result) {b=true}
  n83 = Multiply(A=@n82, B=@n82)
  n84 = VectorParameter() {group=GalaxyV2, parameterName=StarBrightness_Color2, defaultValue=<rgba 950 950 950 1>}
  n85 = ScalarParameter() {defaultValue=1, group=GalaxyV2, parameterName=BrightStar_Tiling, sliderMax=4}
  n86 = NamedRerouteUsage()
  n87 = ComponentMask(None=MaterialFunctionCall!(CameraVector=NamedRerouteUsage(), Texture=@n36, Tiling=Multiply(A=@n85, B=@n86), Offset=@n44.RGB, AxisFadeContrast) {materialFunction=/Engine/Functions/Engine_MaterialFunctions01/Texturing/TriplanarCameraVector}.Result) {r=true}
  n88 = Multiply(A=@n87, B=@n87)
  NamedRerouteDeclaration(None=LinearInterpolate(A=LinearInterpolate(A=@n45.Result, B=@n77, Alpha=@n83), B=Multiply(A=@n76, B=@n84.RGB), Alpha=Multiply(A=@n88, B=@n37.G))) {name=Galaxy, nodeColor=<rgba 0 1 0.7891 1>}
  PinBase!(M=@n22, Pre Skinned Local Position=@n24, Distance Blend=@n25.Distance Blend, Fresnel=@n26) {reroutePins [
    M {expression=&self:MaterialExpressionReroute_4}
    Pre Skinned Local Position {expression=&self:MaterialExpressionReroute_5}
    Distance Blend {expression=&self:MaterialExpressionReroute_6}
    Fresnel {expression=&self:MaterialExpressionReroute_7}
   ]}
  PinBase!()
  TransformPosition(None=@n71) {transformType=TRANSFORMPOSSOURCE_World}
  VertexNormalWS()
 ]
}
